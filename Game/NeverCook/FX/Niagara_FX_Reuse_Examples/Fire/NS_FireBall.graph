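# NS_FireBall is a Niagara fireball effect built from one mesh-particle emitter, Explo_Small, resembling a small explosion puff.
# An instantaneous burst spawns short-lived, randomly sized meshes on a small sphere and pushes them outward, with rising drag slowing them.
# Over lifetime, curves swell then shrink each rotating mesh, cool its color from bright red through near-black to fading grey, and animate dynamic material parameters and camera offset.
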NiagaraSystem /Game/NeverCook/FX/Niagara_FX_Reuse_Examples/Fire/NS_FireBall
dependencies {
 usedDataInterfaces [&NiagaraDataInterfaceCurve, &NiagaraDataInterfaceVectorCurve]
 usedDynamicInputs []
 usedModules [&SystemState]
}
emitters {Explo_Small {
  emitterUpdateScript [
   &EmitterState(Life Cycle Mode=::NewEnumerator1, Loop Behavior=::NewEnumerator1, Loop Duration=1, UseLoopDelay=true)
   &SpawnBurst_Instantaneous(Spawn Count=15, Spawn Time=0.02, Use Spawn Probability=false, UseLoopCountLimit=false)
  ]
  particleSpawnScript [
   &InitializeParticle(
    Lifetime=&RandomRangeFloat(Minimum=0.2, Maximum=1.2)
    Mesh Scale=&VectorFromFloat(Value=&RandomRangeFloat(Minimum=0.03, Maximum=0.07))
    Mesh Scale Mode=::NewEnumerator4
    UsePositionOffset=false
    Write Material Random=true
   )
   &ShapeLocation(
    Hemisphere Distribution=<0.5 1>
    MaskSpawnGroup=false
    Override Seed=false
    Rotation Mode=::NewEnumerator2
    Sphere Radius=10
    UseOwnerScale=true
    Yaw / Pitch / Roll=<0.75 0 0>
   )
   &AddVelocity(Velocity=&Multiply_VectorByFloat(Vector=$Particles.ShapeLocation.ShapeVector, Float=&RandomRangeFloat(Minimum=50, Maximum=70)))
  ]
  particleUpdateScript [
   &ParticleState()
   &UpdateMeshOrientation(Rotation Rate=&RandomRangeFloat(Minimum=-0.5, Maximum=0.5), Rotation Vector=<1 1 1>)
   &ScaleMeshSize(Scale Factor=&VectorFromFloat(Value=&FloatFromCurve(FloatCurve=DataInterface() {propertyValues {Curve {keys [
          <at 0.001 1.806>
          {arriveTangent=5.2789, interpMode=RCIM_Cubic, leaveTangent=5.2789, time=0.047, value=3.0124}
          {arriveTangent=-2.5893, interpMode=RCIM_Cubic, leaveTangent=-2.5893, time=0.2611, value=3.1789}
          {arriveTangent=-3.4302, interpMode=RCIM_Cubic, leaveTangent=-3.4302, time=0.3902, value=2.1239}
          {arriveTangent=-3.4184, interpMode=RCIM_Cubic, leaveTangent=-3.4184, time=0.7005, value=1.6716}
          <at 1.0116 -0.0005>
         ]}}})))
   &Color(Color=&MakeLinearColorFromVectorAndFloat(
     "Vector (RGB)"=&VectorFromCurve(VectorCurve=DataInterface() {propertyValues {
        XCurve {keys [
          <at 0.0082 1.7597 RCIM_Cubic>
          {arriveTangent=-2.8293, interpMode=RCIM_Cubic, leaveTangent=-2.8293, time=0.4127, value=0.0294}
          {arriveTangent=0.135, interpMode=RCIM_Cubic, leaveTangent=0.135, time=0.6199, value=0.0289}
          <at 1.0206 0.1115 RCIM_Cubic>
         ]}
        YCurve {keys [
          <at 0.0082 0.0011 RCIM_Cubic>
          {arriveTangent=0.0262, interpMode=RCIM_Cubic, leaveTangent=0.0262, time=0.4127, value=0.0135}
          {arriveTangent=0.1547, interpMode=RCIM_Cubic, leaveTangent=0.1547, time=0.6199, value=0.0171}
          <at 1.0206 0.1076 RCIM_Cubic>
         ]}
        ZCurve {keys [
          <at 0.0082 0.0026 RCIM_Cubic>
          {arriveTangent=0.0262, interpMode=RCIM_Cubic, leaveTangent=0.0262, time=0.4127, value=0.015}
          {arriveTangent=0.1547, interpMode=RCIM_Cubic, leaveTangent=0.1547, time=0.6199, value=0.0187}
          <at 1.0206 0.1091 RCIM_Cubic>
         ]}
       }})
     "Float (Alpha)"=&ClampFloat(Float=&FloatFromCurve(FloatCurve=DataInterface() {propertyValues {Curve {keys [
           <at -0.0181 0.8475 RCIM_Cubic>
           <at 0.201 0.823>
           {arriveTangent=-1.03, interpMode=RCIM_Cubic, leaveTangent=-1.03, time=0.6554, value=0.8304}
           <at 1 0>
          ]}}}))
    ))
   &Drag(
    Drag=&FloatFromCurve(FloatCurve=DataInterface() {propertyValues {Curve {keys [<at 0.1097 0.1121 RCIM_Cubic>, {interpMode=RCIM_Cubic, leaveTangent=12.8094, time=0.2836, value=6.2478}]}}})
    Use Linear Drag=true
    Use Rotational Drag=true
   )
   &SolveForcesAndVelocity(Clamp Velocity=false, Limit Acceleration=false)
   &DynamicMaterialParameters(
    Index 0 Param 2=&FloatFromCurve(FloatCurve=DataInterface() {propertyValues {Curve {keys [
         <at 0.1351 0.7718 RCIM_Cubic>
         {arriveTangent=-5.6576, interpMode=RCIM_Cubic, leaveTangent=-5.6576, time=0.221, value=0.5477}
         {arriveTangent=-15.2632, interpMode=RCIM_Cubic, leaveTangent=-15.2632, time=0.4316, value=-0.9054}
         {arriveTangent=-30.0299, interpMode=RCIM_Cubic, leaveTangent=-30.0299, time=0.5582, value=-4.5991}
         <at 0.798 -11.9088 RCIM_Cubic>
        ]}}})
    Index 0 Param 3=-0.4
    Index 0 Param 4=&FloatFromCurve(FloatCurve=DataInterface() {propertyValues {Curve {keys [<at -0.088 2.3958 RCIM_Cubic>,<at 0.9925 0.4853 RCIM_Cubic>]}}})
    Module.Index0Param0Write=false
    Module.Index0Param1Write=true
    Module.Index0Param2Write=true
    Module.Index0Param3Write=true
   ) {module=DynamicMaterialParameters001}
   &CameraOffset(Camera Offset Amount=&FloatFromCurve(FloatCurve=DataInterface() {propertyValues {Curve {keys [<at 0 0 RCIM_Cubic>,<at 0.6428 228.4323 RCIM_Cubic>]}}}))
  ]
  rendererClasses [&NiagaraMeshRendererProperties]
  renderers [{rendererClass=&NiagaraMeshRendererProperties, rendererIndex=0}]
 }}
summary {
 systemName=NS_FireBall
 emitters [{emitterName=Explo_Small, rendererClasses [&NiagaraMeshRendererProperties]}]
 userVariables []
}
