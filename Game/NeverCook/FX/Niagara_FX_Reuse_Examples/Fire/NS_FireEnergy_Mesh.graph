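# Niagara system for a fire energy effect: one emitter, Energy01, burst-spawns mesh particles with a reddish HDR, low-alpha tint.
# Particles ignore lifetime expiry and are set to die when the emitter deactivates.
# Spawn sets mesh scale, orientation, material random and dynamic material parameters.
# Each update scales size and color, and writes a smoothed moving average of particle position into an Interpolate Dynamic attribute, giving a lagged position value.
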NiagaraSystem /Game/NeverCook/FX/Niagara_FX_Reuse_Examples/Fire/NS_FireEnergy_Mesh
dependencies {usedDataInterfaces [], usedDynamicInputs [], usedModules [&SystemState]}
emitters {Energy01 {
  emitterUpdateScript [
   &EmitterState(
    Inactive Response=::NewEnumerator1
    Life Cycle Mode=::NewEnumerator1
    Loop Behavior=::NewEnumerator1
    Loop Duration Mode=::NewEnumerator1
    UseLoopDelay=false
   )
   &SpawnBurst_Instantaneous(Use Spawn Probability=false, UseLoopCountLimit=false)
  ]
  particleSpawnScript [
   &InitializeParticle(
    Color=<rgba 3 0.4964 0.4367 0.1>
    Mesh Scale=<1.5 0.5 1>
    Mesh Scale Mode=::NewEnumerator4
    UsePositionOffset=false
    Write Material Random=true
   )
   &DynamicMaterialParameters(
    Index 0 Param 3=20
    Module.Param0WriteEnabled=true
    Module.Param1WriteEnabled=true
    Module.Param2WriteEnabled=true
    Module.Param3WriteEnabled=true
   )
   &InitialMeshOrientation(Mesh Orientation Mode=::NewEnumerator2, Use Rotation Vector=true)
  ]
  particleUpdateScript [
   &ParticleState(Kill Particles When Lifetime Has Elapsed=false, Let Infinitely Lived Particles Die When Emitter Deactivates=true)
   &ScaleMeshSize(Scale Factor=&Multiply_VectorByFloat(Vector=<1 1 1>, Float=0.8))
   &InterpolateOverTime(Rate Of Change=7, Target Vector Value:NiagaraPosition=$Particles.Position, Variable Type=::NewEnumerator2)
   /SetVariables(Particles.Interpolate Dynamic=&MakeFloat4FromVectorAndFloat("Vector (RGB)"=$Output.InterpolateOverTime.Moving Average, "Float (Alpha)"=1)) {isSetParameters=true}
   &ScaleColor(Scale Alpha=0.5, ScaleA=true, ScaleRGB=true)
  ]
  rendererClasses [&NiagaraMeshRendererProperties]
  renderers [{rendererClass=&NiagaraMeshRendererProperties, rendererIndex=0}]
 }}
summary {
 systemName=NS_FireEnergy_Mesh
 emitters [{emitterName=Energy01, rendererClasses [&NiagaraMeshRendererProperties]}]
 userVariables []
}
