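# Niagara fire ring effect, from a folder of reusable FX examples, built from a single mesh-particle emitter, FireRing.
# Meshes spawn at a steady rate with a fixed rotation plus a random spin on one axis, so short-lived copies overlap.
# Each update tints them within an orange-to-yellow range and fades alpha in and out over life.
# Dynamic material parameters feed random values, constants and a rise-and-fall curve to the mesh material.
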NiagaraSystem /Game/NeverCook/FX/Niagara_FX_Reuse_Examples/Fire/NS_Fire_Ring
dependencies {usedDataInterfaces [&NiagaraDataInterfaceCurve], usedDynamicInputs [], usedModules [&SystemState]}
emitters {FireRing {
  emitterUpdateScript [
   &EmitterState(Inactive Response=::NewEnumerator1, Life Cycle Mode=::NewEnumerator1, Loop Duration=1, UseLoopDelay=false)
   &SpawnRate(SpawnRate=6, Use Spawn Probability=false)
  ]
  particleSpawnScript [
   &InitializeParticle(Lifetime=3, Mesh Scale=<-1.5 2 2>, Mesh Scale Mode=::NewEnumerator4, UsePositionOffset=false, Write Material Random=true)
   &InitialMeshOrientation(Mesh Orientation Mode=::NewEnumerator2, Rotation=<0 0.25 0>, Use Rotation Vector=true)
   &InitialMeshOrientation(
    Mesh Orientation Mode=::NewEnumerator2
    Rotation=&RandomRangeVector(Minimum=<0 0 0>, Maximum=<0 0 1>)
    Use Rotation Vector=true
   ) {module=InitialMeshOrientation001}
  ]
  particleUpdateScript [
   &ParticleState()
   &ScaleColor(
    Scale Alpha=&FloatFromCurve(FloatCurve=DataInterface() {propertyValues {Curve {keys [<at 0 0 RCIM_Cubic>,<at 0.3 1 RCIM_Cubic>,<at 1 0 RCIM_Cubic>]}}})
    Scale RGB=&MakeVectorFromLinearColorRGB(Color=&RandomRangeLinearColor(Minimum=<rgba 1 0.1409 0.065 1>, Maximum=<rgba 1.5 1.2628 0 1>))
    ScaleA=true
    ScaleRGB=true
   )
   &DynamicMaterialParameters(
    Index 0 Param 1=&RandomRangeFloat(Minimum=1, Maximum=2)
    Index 0 Param 2=3
    Index 0 Param 3=3
    Index 0 Param 4=&RandomRangeFloat(Minimum=1, Maximum=1.6)
    Index 2 Param 1=&FloatFromCurve(
     FloatCurve=DataInterface() {propertyValues {Curve {keys [
         {arriveTangent=-0.01, interpMode=RCIM_Cubic, leaveTangent=-0.01, tangentMode=RCTM_User, time=0, value=0}
         {arriveTangent=2.1454, interpMode=RCIM_Cubic, leaveTangent=2.1454, tangentMode=RCTM_User, time=0.2, value=0.2}
         {arriveTangent=0.0029, interpMode=RCIM_Cubic, leaveTangent=0.0029, tangentMode=RCTM_User, time=0.3, value=0.4}
         {arriveTangent=-0.7109, interpMode=RCIM_Cubic, leaveTangent=-0.7109, tangentMode=RCTM_User, time=0.5, value=0.175}
         {arriveTangent=-0.0738, interpMode=RCIM_Cubic, leaveTangent=-0.0738, tangentMode=RCTM_User, time=1, value=0}
        ]}}}
     Scale Curve=&Multiply_Float(A=1, B=2)
    )
    Index 2 Param 2=1
    Index 2 Param 3=1
    Index 2 Param 4=1
    Module.Index0Float0Write=true
    Module.Index0Float1Write=true
    Module.Index0Float2Write=true
    Module.Index0Float3Write=true
    Module.Index2Float0Write=true
    Module.Index2Float1Write=false
    Module.Index2Float2Write=false
    Module.Index2Float3Write=false
    Write Parameter Index 2=::NewEnumerator1
   )
  ]
  rendererClasses [&NiagaraMeshRendererProperties]
  renderers [{rendererClass=&NiagaraMeshRendererProperties, rendererIndex=0}]
 }}
summary {
 systemName=NS_Fire_Ring
 emitters [{emitterName=FireRing, rendererClasses [&NiagaraMeshRendererProperties]}]
 userVariables []
}
