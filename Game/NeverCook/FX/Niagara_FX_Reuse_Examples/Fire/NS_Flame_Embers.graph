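# Niagara system for small flame embers, one sprite emitter with a steady, low spawn rate.
# Short-lived sprites spawn in a small offset sphere, rise along Z and drift under curl noise while curves swell then shrink their size and fade alpha in and out.
# User.VariantID indexes a color array to pick orange-gold or green-teal HDR embers.
# User.IsSkydiving is declared, but no module shown reads it.
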NiagaraSystem /Game/NeverCook/FX/Niagara_FX_Reuse_Examples/Fire/NS_Flame_Embers
dependencies {
 usedDataInterfaces [&NiagaraDataInterfaceArrayColor, &NiagaraDataInterfaceArrayFloat, &NiagaraDataInterfaceCurve]
 usedDynamicInputs [&Multiply_Float, &SelectFloatFromArray]
 usedModules [&SystemState]
}
emitters {NE_Embers {
  emitterUpdateScript [
   &EmitterState()
   &SpawnRate(SpawnRate=10, Use Spawn Probability=true)
  ]
  particleSpawnScript [
   &InitializeParticle(
    Color=&SelectLinearColorFromArray(
     Array Sampling Mode=::NewEnumerator1
     Direct Array Index=&MakeIntFromFloat(Float=$User.VariantID)
     Color Selection Array=DataInterface() {propertyValues {ColorData [<rgba 2 0.7121 0 0.9014>, <rgba 0.4592 2 1.2296 0.9014>]}}
     Random Seed=?
    )
    Lifetime Max=1
    Lifetime Min=0.5
    Lifetime Mode=::NewEnumerator1
    Position Offset=<0 -17.5 -5>
    Position Offset Coordinate Space=::Local
    Sprite Rotation Mode=::NewEnumerator2
    Sprite Size Mode=::NewEnumerator1
    Uniform Sprite Size Max=8
    Uniform Sprite Size Min=4
    UsePositionOffset=true
    Write Material Random=true
   )
   &ShapeLocation(MaskSpawnGroup=false, Override Seed=false, Sphere Radius=10, UseOwnerScale=true)
   &AddVelocity(
    Cone Angle=0
    Cone Axis=<0 0 1>
    Module.UseConeFalloff=false
    Module.UseVelDistribution=false
    Override Seed=false
    Velocity Mode=::NewEnumerator2
    Velocity Speed=&RandomRangeFloat(Minimum=15, Maximum=30)
   )
  ]
  particleUpdateScript [
   &ParticleState()
   &ScaleSpriteSize(Uniform Curve Sprite Scale=DataInterface() {propertyValues {Curve {
       defaultValue=0
       keys [
        {arriveTangent=4.8998, interpMode=RCIM_Cubic, leaveTangent=4.8998, tangentMode=RCTM_User, time=0.0003, value=0.0536}
        {interpMode=RCIM_Cubic, tangentMode=RCTM_User, time=0.2908, value=0.9741}
        {arriveTangent=-4.9964, interpMode=RCIM_Cubic, leaveTangent=-4.9964, tangentMode=RCTM_User, time=1, value=0}
       ]
      }}})
   &ScaleColor(
    Scale Alpha=&FloatFromCurve(FloatCurve=DataInterface() {propertyValues {Curve {
        defaultValue=1
        keys [
         {arriveTangent=9.0725, interpMode=RCIM_Cubic, leaveTangent=9.0725, tangentMode=RCTM_User, time=0, value=0}
         {arriveTangent=0.5892, interpMode=RCIM_Cubic, leaveTangent=0.5892, tangentMode=RCTM_User, time=0.1408, value=0.9847}
         {arriveTangent=-2.1532, interpMode=RCIM_Cubic, leaveTangent=-2.1532, tangentMode=RCTM_User, time=1, value=0}
        ]
       }}})
    ScaleA=true
    ScaleRGB=true
   )
   &CurlNoiseForce(
    Bias Noise=false
    Noise Frequency=200
    Noise Strength=&RandomRangeFloat(Minimum=50, Maximum=100)
    Pan Noise=false
   )
   &SolveForcesAndVelocity(Clamp Velocity=false, Limit Acceleration=false)
  ]
  rendererClasses [&NiagaraSpriteRendererProperties]
  renderers [{rendererClass=&NiagaraSpriteRendererProperties}]
 }}
summary {
 systemName=NS_Flame_Embers
 emitters [{emitterName=NE_Embers, rendererClasses [&NiagaraSpriteRendererProperties]}]
 userVariables [
  User.IsSkydiving {type=NiagaraBool}
  User.VariantID {type=NiagaraFloat, defaultValue=0, description=""}
 ]
}
