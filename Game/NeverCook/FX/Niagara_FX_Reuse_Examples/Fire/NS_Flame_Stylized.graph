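# NS_Flame_Stylized is a Niagara system for a stylized flame, built from one sprite emitter in the reuse examples folder.
# An instantaneous burst on a short loop spawns tall, world-Z-aligned sprites that live about one loop, with varied size and material random values.
# A jagged float curve scales sprite height to create flicker while width stays constant.
# Two front-end user booleans are declared, with no visible use.
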NiagaraSystem /Game/NeverCook/FX/Niagara_FX_Reuse_Examples/Fire/NS_Flame_Stylized
dependencies {
 usedDataInterfaces [&NiagaraDataInterfaceCurve, &NiagaraDataInterfaceVector2DCurve]
 usedDynamicInputs []
 usedModules [&SystemState]
}
emitters {Flame001 {
  emitterUpdateScript [
   &EmitterState(Life Cycle Mode=::NewEnumerator1, Loop Duration=0.3, UseLoopDelay=false)
   &SpawnBurst_Instantaneous(Use Spawn Probability=false, UseLoopCountLimit=false)
  ]
  particleSpawnScript [
   &InitializeParticle(
    Color Mode=::NewEnumerator0
    Lifetime Max=0.31
    Lifetime Min=0.3
    Lifetime Mode=::NewEnumerator1
    Sprite Size Max=<15 50>
    Sprite Size Min=<13 45>
    Sprite Size Mode=::NewEnumerator2
    UsePositionOffset=false
    Write Material Random=true
   )
   &SpriteFacingAndAlignment(Alignment Coordinate Space=::World, Sprite Alignment=<0 0 1>, WriteAlignment=true, WriteFacing=false) {module=SpriteFacingAndAlignment001}
  ]
  particleUpdateScript [
   &ParticleState()
   &ScaleSpriteSize(
    Non-Uniform Curve Index:NiagaraFloat=$System.NormalizedLoopAge
    Non-Uniform Curve Scale=&MakeVector2D(
     X=1
     Y=&FloatFromCurve(FloatCurve=DataInterface() {propertyValues {Curve {keys [<at 0 1>,<at 0.0985 0.8507>,<at 0.2008 1.0157>,<at 0.26 0.8488>,<at 0.3458 1.0027>,<at 0.3762 0.8539>,<at 0.4547 0.9889>,<at 0.5653 0.8593>,<at 0.6393 0.9746>,<at 0.7324 0.8661>,<at 0.8044 1.0027>,<at 0.9049 0.8638>,<at 0.9964 0.9311>]}}})
    )
    Non-Uniform Curve Sprite Scale=DataInterface() {propertyValues {
      XCurve {keys [{tangentMode=RCTM_User, time=0, value=1}, {tangentMode=RCTM_User, time=1, value=1}]}
      YCurve {keys [{tangentMode=RCTM_User, time=0, value=1}, {tangentMode=RCTM_User, time=1, value=1}]}
     }}
    Scale Sprite Size Mode=::NewEnumerator3
   )
  ]
  rendererClasses [&NiagaraSpriteRendererProperties]
  renderers [{rendererClass=&NiagaraSpriteRendererProperties}]
 }}
summary {
 systemName=NS_Flame_Stylized
 emitters [{emitterName=Flame001, rendererClasses [&NiagaraSpriteRendererProperties]}]
 userVariables [
  User.bIsFrontEnd {type=NiagaraBool}
  User.bIsFrontEndPreview {type=NiagaraBool}
 ]
}
