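# Niagara mesh-flame effect for a nitro speed fire, stored among NeverCook's reusable fire FX examples.
# Its single emitter, Flame_Core_Meshes, bursts short-lived, camera-offset meshes, using spawn-group masks to vary color, rotation rate and alpha fade.
# Curves grow the meshes early in life and fade them out late, while random and curve-driven dynamic parameters feed the material.
# User.ControlCurve is exposed, but no listed module input reads it.
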NiagaraSystem /Game/NeverCook/FX/Niagara_FX_Reuse_Examples/Fire/NS_Nitro_SpeedFire
dependencies {
 usedDataInterfaces [&NiagaraDataInterfaceCurve, &NiagaraDataInterfaceVectorCurve]
 usedDynamicInputs []
 usedModules [&SystemState]
}
emitters {Flame_Core_Meshes {
  emitterUpdateScript [
   &EmitterState()
   &SpawnBurst_Instantaneous(Spawn Time=0.05, Use Spawn Probability=false, UseLoopCountLimit=false)
  ]
  particleSpawnScript [
   &InitializeParticle(
    Color=&MaskLinearColorBySpawnGroup(Masked Linear Color=<rgba 3 3 3 0.8>, Passthrough Linear Color=<rgba 3 2.5508 1.8811 0.8>)
    Lifetime=0.5
    Mesh Scale=<3 2 2>
    Mesh Scale Mode=::NewEnumerator4
    UsePositionOffset=true
    Write Material Random=true
   )
   &CameraOffset(Camera Offset Amount=50)
   &InitialMeshOrientation(Mesh Orientation Mode=::NewEnumerator2, Use Rotation Vector=false)
  ]
  particleUpdateScript [
   &ParticleState()
   &UpdateMeshOrientation(Rotation Rate=&MaskFloatBySpawnGroup(Masked Float=0, Passthrough Float=1))
   &DynamicMaterialParameters(
    Index 0 Param 1=&RandomRangeFloat(Minimum=0, Maximum=1)
    Index 0 Param 2=0.125
    Index 0 Param 3=2
    Index 0 Param 4=&FloatFromCurve(FloatCurve=DataInterface() {propertyValues {Curve {keys [<at 0.75 1>,<at 1 0>]}}}, Scale Curve=250)
    Index 2 Param 2=&RandomRangeFloat(Minimum=0.25, Maximum=1.25)
    Module.Index0Param0Write=true
    Module.Index0Param1Write=true
    Module.Index0Param2Write=true
    Module.Index0Param3Write=true
    Module.Index2Param0Write=true
    Module.Index2Param1Write=true
    Module.Index2Param2Write=false
    Module.Index2Param3Write=false
    Write Parameter Index 2=true
   )
   &ScaleMeshSize(Scale Factor=&VectorFromCurve(VectorCurve=DataInterface() {propertyValues {
       XCurve {keys [
         <at 0 0 RCIM_Cubic>
         {arriveTangentWeight=0.7538, interpMode=RCIM_Cubic, leaveTangentWeight=1.6584, tangentMode=RCTM_User, time=0.25, value=1}
        ]}
       YCurve {keys [{arriveTangent=3.1731, interpMode=RCIM_Cubic, leaveTangent=3.1731, tangentMode=RCTM_User, time=0, value=0.25}, <at 0.25 1 RCIM_Cubic>]}
       ZCurve {keys [{arriveTangent=3.1731, interpMode=RCIM_Cubic, leaveTangent=3.1731, tangentMode=RCTM_User, time=0, value=0.25}, <at 0.25 1 RCIM_Cubic>]}
      }}))
   &ScaleColor(
    Scale Alpha=&MaskFloatBySpawnGroup(
     Masked Float=&FloatFromCurve(FloatCurve=DataInterface() {propertyValues {Curve {keys [<at 0.8 1 RCIM_Cubic>,<at 1 0 RCIM_Cubic>]}}})
     Passthrough Float=&FloatFromCurve(FloatCurve=DataInterface() {propertyValues {Curve {keys [<at 0.8 1 RCIM_Cubic>, {arriveTangent=-16.5342, interpMode=RCIM_Cubic, leaveTangent=-16.5342, tangentMode=RCTM_User, time=1, value=0}]}}})
    )
    ScaleA=true
    ScaleRGB=true
   )
  ]
  rendererClasses [&NiagaraMeshRendererProperties]
  renderers [{rendererClass=&NiagaraMeshRendererProperties, rendererIndex=0}]
 }}
summary {
 systemName=NS_Nitro_SpeedFire
 emitters [{emitterName=Flame_Core_Meshes, rendererClasses [&NiagaraMeshRendererProperties]}]
 userVariables [User.ControlCurve {type=NiagaraFloat, defaultValue=0, description=""}]
}
