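# Niagara system for a pickaxe torch flame, built from one mesh-based emitter, Fire_Spline.
# A burst spawns persistent particles with orange HDR colors, separate size and glow-size attributes, and randomized material values, drawn by two mesh renderers.
# Each update smooths particle position over time and packs that moving average with a constant alpha into an Interpolate Dynamic attribute, likely so the material can lag the flame behind motion.
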NiagaraSystem /Game/NeverCook/FX/Niagara_FX_Reuse_Examples/Fire/NS_Pickaxe_Torch_Flame
dependencies {usedDataInterfaces [], usedDynamicInputs [], usedModules [&SystemState]}
emitters {Fire_Spline {
  emitterUpdateScript [
   &EmitterState(Life Cycle Mode=::NewEnumerator1, Loop Behavior=::NewEnumerator1, Loop Duration Mode=::NewEnumerator1, UseLoopDelay=false)
   &SpawnBurst_Instantaneous(Use Spawn Probability=false, UseLoopCountLimit=false)
  ]
  particleSpawnScript [
   &InitializeParticle(
    Color=<rgba 2 0.6989 0.0492 1>
    Lifetime=5
    Mesh Scale=<0.08 0.035 0.035>
    Mesh Scale Mode=::NewEnumerator4
    UsePositionOffset=true
    Write Material Random=true
   )
   /SetVariables(Particles.Size003:Vector3f=&Multiply_Vector(A=<2.5 2.5 2.5>, B=&VectorFromFloat(Value=1))) {isSetParameters=true}
   /SetVariables(Particles.Glow Size:Vector3f=&Multiply_Vector(A=<2.8 2.8 2.8>, B=&VectorFromFloat(Value=1))) {isSetParameters=true}
   /SetVariables(Particles.Color002=<rgba 2.2 0.95 0.24 1.2>) {isSetParameters=true}
   /SetVariables(Particles.Color003:LinearColor=<rgba 2 0.43 0 0.4>) {isSetParameters=true}
   &InitialMeshOrientation(
    Mesh Axis To Orient=<0 0 1>
    Mesh Orientation Mode=::NewEnumerator2
    Orientation Vector=<0 0 1>
    Rotation=<0 0 0.25>
    Use Rotation Vector=true
   )
   &InitialMeshOrientation(Mesh Orientation Mode=::NewEnumerator2, Use Rotation Vector=true) {module=InitialMeshOrientation001}
  ]
  particleUpdateScript [
   &ParticleState(Kill Particles When Lifetime Has Elapsed=false, Let Infinitely Lived Particles Die When Emitter Deactivates=true)
   &InterpolateOverTime(Rate Of Change=8, Target Vector Value:NiagaraPosition=$Particles.Position, Variable Type=::NewEnumerator2)
   /SetVariables(Particles.Interpolate Dynamic=&MakeFloat4FromVectorAndFloat("Vector (RGB)"=$Particles.InterpolateOverTime.InterpolateOverTime_V3_.Moving Average, "Float (Alpha)"=0.7)) {isSetParameters=true}
  ]
  rendererClasses [&NiagaraMeshRendererProperties]
  renderers [
   {rendererClass=&NiagaraMeshRendererProperties, rendererIndex=0}
   {rendererClass=&NiagaraMeshRendererProperties, rendererIndex=1}
  ]
 }}
summary {
 systemName=NS_Pickaxe_Torch_Flame
 emitters [{emitterName=Fire_Spline, rendererClasses [&NiagaraMeshRendererProperties]}]
 userVariables [
  User.bIsFrontEndPreview {type=NiagaraBool}
  User.FXScale {type=NiagaraFloat, defaultValue=1, description=""}
 ]
}
